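# Control Rig for the M_MED_NotchGnarl player body that adds procedural corrective deformation and state-based bone tweaks.
# A shared DeformRig function runs first, then parent constraints blend arm-fur and rear-foot helper bones.
# Spherical pose readers on both feet drive rear-foot correctives.
# An eased ParachuteOpened adjusts upper-arm twists, Front End with BackBiling adjusts the tail, and Custom_Emote shrinks feet and collapses the tail.
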ControlRigBlueprint /BRCosmetics/Characters/Player/Male/Medium/Bodies/M_MED_NotchGnarl/Meshes/M_MED_NotchGnarl_CtrlRig
graphs [
 New Function_ContainedGraph {nodes [
   n1 = RigVMFunctionEntryNode()
   RigVMFunction_MathBoolToInteger = &RigVMFunction_MathBoolToInteger(Value=@n1.onlyTwist)
   RigVMFunction_MathQuaternionSwingTwist = &RigVMFunction_MathQuaternionSwingTwist(
    twistAxis=@n1.twistAxis
    Input=&RigUnit_GetRelativeTransformForItem(bChildInitial=False, bParentInitial=False, child=@n1.child, parent=@n1.parent).RelativeTransform.Rotation
   )
   n2 = RigVMDispatchNode(
    Values=((),())
    Index=@RigVMFunction_MathBoolToInteger.Result
    Values.0=@RigVMFunction_MathQuaternionSwingTwist.Swing
    Values.1=@RigVMFunction_MathQuaternionSwingTwist.Twist
   ) {template="DISPATCH_RigVMDispatch_SelectInt32(in Index,in Values,out Result)"}
   RigVMFunction_MathQuaternionToEuler = &RigVMFunction_MathQuaternionToEuler(Value=@n2.Result, RotationOrder=@n1.Rotation Order)
   > DISPATCH_RigVMDispatch_Print = RigVMDispatchNode(ScreenColor=(R=1,G=1,B=1,A=1), ScreenDuration=1.5, Enabled=@n1.debugRotation, Value=@RigVMFunction_MathQuaternionToEuler.Result) {template="DISPATCH_RigVMDispatch_Print(in Prefix,in Value,in Enabled,in ScreenDuration,in ScreenColor)"}
   Cast = RigVMDispatchNode(Value=@n1.triggerAxis) {template="DISPATCH_RigVMDispatch_CastEnumToInt(in Value,out Result)"}
   DISPATCH_RigVMDispatch_SelectInt32 = RigVMDispatchNode(
    Values=(0,0,0)
    Index=@Cast.Result
    Values.0=@RigVMFunction_MathQuaternionToEuler.Result.X
    Values.1=@RigVMFunction_MathQuaternionToEuler.Result.Y
    Values.2=@RigVMFunction_MathQuaternionToEuler.Result.Z
   ) {template="DISPATCH_RigVMDispatch_SelectInt32(in Index,in Values,out Result)"}
   Remap = &RigVMFunction_MathDoubleRemap(
    bClamp=False
    Value=@DISPATCH_RigVMDispatch_SelectInt32.Result
    TargetMinimum=@n1.outputMinMax.X
    TargetMaximum=@n1.outputMinMax.Y
    SourceMinimum=@n1.rotationMinMax.X
    SourceMaximum=@n1.rotationMinMax.Y
   )
   Absolute = &RigVMFunction_MathDoubleAbs(Value=@Remap.Result)
   Multiply = &RigVMFunction_MathDoubleMul(
    A=&RigVMFunction_MathDoubleClamp(Value=RigVMDispatchNode(True=@Absolute.Result, Condition=@n1.absoluteValue, False=@Remap.Result) {template="DISPATCH_RigVMDispatch_If(in Condition,in True,in False,out Result)"}.Result).Result
    B=@n1.premult
   )
   > RigVMFunctionReturnNode(outputQuat=@n2.Result, outputWeight=@Multiply.Result)
  ]}
 New Function_ContainedGraph {nodes [
   RigVMFunctionEntryNode()
   > &RigUnit_ParentConstraint(
    AdvancedSettings=(RotationOrderForFilter=XZY)
    Child=(Type=Bone,Name="base_ArmFur_l_01")
    Parents="((Item=(Type=Bone,Name=\"lowerarm_twist_01_l\"),Weight=0.0124),(Item=(Type=Bone,Name=\"upperarm_l\"),Weight=0.0181),(Item=(Type=Bone,Name=\"lowerarm_l\"),Weight=0.108),(Item=(Type=Bone,Name=\"lowerarm_twist_02_l\"),Weight=0.1186),(Item=(Type=Bone,Name=\"deform_elbo ...(truncated)"
    Weight=1
   )
   > &RigUnit_ParentConstraint(
    AdvancedSettings=(RotationOrderForFilter=XZY)
    Child=(Type=Bone,Name="base_ArmFur_l_02")
    Parents="((Item=(Type=Bone,Name=\"upperarm_twist_02_l\"),Weight=0.0111),(Item=(Type=Bone,Name=\"lowerarm_twist_02_l\"),Weight=0.0428),(Item=(Type=Bone,Name=\"lowerarm_l\"),Weight=0.0639),(Item=(Type=Bone,Name=\"upperarm_l\"),Weight=0.0723),(Item=(Type=Bone,Name=\"deform_elb ...(truncated)"
    Weight=1
   )
   > &RigUnit_ParentConstraint(
    AdvancedSettings=(RotationOrderForFilter=XZY)
    Child=(Type=Bone,Name="base_ArmFur_l_03")
    Parents="((Item=(Type=Bone,Name=\"lowerarm_twist_02_l\"),Weight=0.0253),(Item=(Type=Bone,Name=\"upperarm_twist_02_l\"),Weight=0.0276),(Item=(Type=Bone,Name=\"lowerarm_l\"),Weight=0.0732),(Item=(Type=Bone,Name=\"upperarm_l\"),Weight=0.1832),(Item=(Type=Bone,Name=\"deform_elb ...(truncated)"
    Weight=1
   )
   > &RigUnit_ParentConstraint(
    AdvancedSettings=(RotationOrderForFilter=XZY)
    Child=(Type=Bone,Name="base_ArmFur_r_01")
    Parents="((Item=(Type=Bone,Name=\"lowerarm_twist_01_r\"),Weight=0.0164),(Item=(Type=Bone,Name=\"upperarm_r\"),Weight=0.0245),(Item=(Type=Bone,Name=\"upperarm_twist_02_r\"),Weight=0.1128),(Item=(Type=Bone,Name=\"lowerarm_r\"),Weight=0.1484),(Item=(Type=Bone,Name=\"lowerarm_t ...(truncated)"
    Weight=1
   )
   > &RigUnit_ParentConstraint(
    AdvancedSettings=(RotationOrderForFilter=XZY)
    Child=(Type=Bone,Name="base_ArmFur_r_03")
    Parents="((Item=(Type=Bone,Name=\"lowerarm_twist_02_r\"),Weight=0.0143),(Item=(Type=Bone,Name=\"upperarm_twist_02_r\"),Weight=0.0293),(Item=(Type=Bone,Name=\"lowerarm_r\"),Weight=0.0578),(Item=(Type=Bone,Name=\"upperarm_r\"),Weight=0.1821),(Item=(Type=Bone,Name=\"dyn_ArmFur ...(truncated)"
    Weight=1
   )
   > &RigUnit_ParentConstraint(
    AdvancedSettings=(RotationOrderForFilter=XZY)
    Child=(Type=Bone,Name="base_ArmFur_r_02")
    Parents="((Item=(Type=Bone,Name=\"lowerarm_twist_02_r\"),Weight=0.0458),(Item=(Type=Bone,Name=\"upperarm_twist_02_r\"),Weight=0.0777),(Item=(Type=Bone,Name=\"lowerarm_r\"),Weight=0.0825),(Item=(Type=Bone,Name=\"upperarm_r\"),Weight=0.1009),(Item=(Type=Bone,Name=\"dyn_ArmFur ...(truncated)"
    Weight=1
   )
   > &RigUnit_ParentConstraint(
    AdvancedSettings=(InterpolationType=Average,RotationOrderForFilter=XZY)
    Child=(Type=Bone,Name="deform_foot_base_bk_l")
    Parents=((Item=(Type=Bone,Name="foot_l"),Weight=0.5),(Item=(Type=Bone,Name="calf_l"),Weight=0.5))
    Weight=1
   )
   > &RigUnit_ParentConstraint(
    AdvancedSettings=(InterpolationType=Average,RotationOrderForFilter=XZY)
    Child=(Type=Bone,Name="deform_foot_base_bk_r")
    Parents=((Item=(Type=Bone,Name="foot_r"),Weight=0.5),(Item=(Type=Bone,Name="calf_r"),Weight=0.5))
    Weight=1
   )
   > RigVMFunctionReturnNode()
  ]}
 RigVMFunctionLibrary {nodes [
   Constraints = RigVMCollapseNode(ExecuteContext=())
   DeltaRotation = RigVMCollapseNode(
    ExecuteContext=()
    Rotation Order=ZYX
    child=(Type=Bone,Name="None")
    outputMinMax=(X=0,Y=1)
    parent=(Type=Bone,Name="None")
    premult=1
    rotationMinMax=(X=0,Y=0)
    twistAxis=(X=1,Y=0,Z=0)
   )
  ]}
 RigVMModel {nodes [
   RigUnit_BeginExecution = &RigUnit_BeginExecution()
   FN_DeformRig_FNC = RigVMFunctionReferenceNode(
    DeformRig_ Curve="(deformRig_l_upperarm_up_alpha_72_#1=False,deformRig_r_upperarm_up_alpha_73_#2=False,deformRig_l_upperarm_fr_alpha_74_#3=False,deformRig_r_upperarm_fr_alpha_75_CB51B8 ...(truncated)"
    DeformRig_Bool="(upperarm_1_#4=True,elbow_3_#5=True,wrist_7_#6=True,groin_8_#7=True,glute_17_#8=True,knee_9_40C408324CA261 ...(truncated)"
    DeformRig_Debug=(upperarm_debug_18_#4=False,wrist_debug_20_#6=False,lat_debug_26_#9=False,groinGlute_debug_27_#7=False)
    DeformRig_Value="(upperarm_up_18_#10=(Rotation=(X=0.0,Y=-0.0,Z=0.0,W=1.0),Translation=(X=0.0,Y=0.0,Z=4),Scale3D=(X=0.70,Y=1.0,Z=1.0)),upperarm_fwd_fr_19_#11=(Rotation=(X=0.0,Y=-0.0,Z=0.0,W=1.0),Translation=(X=0.0,Y= ...(truncated)"
    ExecuteContext=@RigUnit_BeginExecution.ExecutePin
   )
   RigVMFunction_Sequence = &RigVMFunction_Sequence(ExecuteContext=@FN_DeformRig_FNC.ExecuteContext)
   Constraints = RigVMFunctionReferenceNode(ExecuteContext=@RigVMFunction_Sequence.A)
   > n1 = &RigUnit_SphericalPoseReader(
    ActiveRegionScaleFactors=(PositiveWidth=1,NegativeWidth=1.2,PositiveHeight=1.5,NegativeHeight=1.5)
    ActiveRegionSize=0.1
    Debug=(bDrawDebug=false,bDraw2D=False,bDrawLocalAxes=False,DebugScale=25,DebugSegments=20,DebugThickness=0.25)
    DriverAxis=(X=-1,Y=0,Z=0)
    DriverItem=(Type=Bone,Name="foot_l")
    FalloffRegionScaleFactors=(PositiveWidth=1,NegativeWidth=2,PositiveHeight=1.5,NegativeHeight=1.5)
    FalloffSize=0.23
    FlipHeightScaling=False
    FlipWidthScaling=False
    OptionalParentItem=(Type=Bone,Name="None")
    RotationOffset=(X=-90,Y=0,Z=-12)
   )
   > &RigUnit_ModifyTransforms(
    ItemToModify=((Item=(Type=Bone,Name="deform_foot_bk_l"),Transform=(Rotation=(X=0,Y=0,Z=0,W=1),Translation=(X=3.2303,Y=-4.3087,Z=0),Scale3D=(X=1,Y=1,Z=1))))
    WeightMinimum=0
    Weight=@n1.OutputParam
   )
   > n2 = &RigUnit_SphericalPoseReader(
    ActiveRegionScaleFactors=(PositiveWidth=1,NegativeWidth=1.2,PositiveHeight=1.5,NegativeHeight=1.5)
    ActiveRegionSize=0.1
    Debug=(bDrawDebug=false,bDraw2D=False,bDrawLocalAxes=False,DebugScale=25,DebugSegments=20,DebugThickness=0.25)
    DriverAxis=(X=1,Y=0,Z=0)
    DriverItem=(Type=Bone,Name="foot_r")
    FalloffRegionScaleFactors=(PositiveWidth=1,NegativeWidth=2,PositiveHeight=1.5,NegativeHeight=1.5)
    FalloffSize=0.23
    FlipHeightScaling=False
    FlipWidthScaling=true
    OptionalParentItem=(Type=Bone,Name="None")
    RotationOffset=(X=90,Y=0,Z=-12)
   )
   > &RigUnit_ModifyTransforms(
    ItemToModify=((Item=(Type=Bone,Name="deform_foot_bk_r"),Transform=(Rotation=(X=0,Y=0,Z=0,W=1),Translation=(X=3.2303,Y=-4.3087,Z=0),Scale3D=(X=1,Y=1,Z=1))))
    WeightMinimum=0
    Weight=@n2.OutputParam
   )
   RigVMFunction_MathBoolToFloat = &RigVMFunction_MathBoolToFloat(Value=RigVMVariableNode() {variable=ParachuteOpened}.Value)
   AlphaInterp = &RigVMFunction_AlphaInterp(InterpSpeedDecreasing=5, InterpSpeedIncreasing=5, Scale=1, bClampResult=False, Value=@RigVMFunction_MathBoolToFloat.Result)
   > &RigUnit_ModifyTransforms(
    ItemToModify="((Item=(Type=Bone,Name=\"upperarm_twist_02_r\"),Transform=(Rotation=(X=0,Y=0,Z=0.0436,W=0.999),Translation=(X=0,Y=1,Z=0),Scale3D=(X=1,Y=1.5,Z=1))),(Item=(Type=Bone,Name=\"upperarm_twist_02_l\"),Transform=(Rotation=(X=0,Y=0,Z=0.0436,W=0.999),Translation=(X=0,Y= ...(truncated)"
    WeightMinimum=0
    Weight=@AlphaInterp.Result
   )
   Multiply = &RigVMFunction_MathDoubleMul(
    B=RigVMDispatchNode(False=0, True=1, Condition=RigVMVariableNode() {variable=BackBiling}.Value) {template="DISPATCH_RigVMDispatch_If(in Condition,in True,in False,out Result)"}.Result
    A=&RigVMFunction_MathBoolToFloat(Value=RigVMVariableNode() {variable=Front End}.Value).Result
   )
   > &RigUnit_ModifyTransforms(
    ItemToModify=((Item=(Type=Bone,Name="dyn_tail_01"),Transform=(Rotation=(X=0,Y=0.0436,Z=0,W=0.999),Translation=(X=0,Y=0,Z=0),Scale3D=(X=1,Y=0.9,Z=1))))
    WeightMinimum=0
    Weight=@Multiply.Result
   )
   n3 = RigVMVariableNode() {variable=Custom_Emote}
   n4 = &RigVMFunction_AlphaInterp(InterpSpeedDecreasing=10, InterpSpeedIncreasing=10, Scale=1, bClampResult=False, Value=@n3.Value)
   &RigUnit_ModifyTransforms(
    ItemToModify=((Item=(Type=Bone,Name="foot_l"),Transform=(Rotation=(X=0,Y=0,Z=0,W=1),Translation=(X=0,Y=0,Z=0),Scale3D=(X=0.5,Y=0.5,Z=0.5))),(Item=(Type=Bone,Name="foot_r"),Transform=(Rotation=(X=0,Y=0,Z=0,W=1),Translation=(X=0,Y=0,Z=0),Scale3D=(X=0.5,Y=0.5,Z=0.5))))
    WeightMinimum=0
    Weight=@n4.Result
    ExecutePin=@RigVMFunction_Sequence.B
   )
   > &RigUnit_ModifyTransforms(
    ItemToModify=((Item=(Type=Bone,Name="base_tail_01"),Transform=(Rotation=(X=0,Y=0,Z=0,W=1),Translation=(X=0,Y=0,Z=0),Scale3D=(X=0,Y=0,Z=0))))
    WeightMinimum=0
    Weight=@n3.Value
   )
  ]}
]
hierarchy [
 root {loc=<0 0 0>, rot=<0 0 0 1>}
 pelvis {parent=root, loc=<0 -0.7071 93.1191>, rot=<0.0127 0.707 -0.0127 -0.707>}
 spine_01 {parent=pelvis, loc=<6.7494 -0.6388 0>, rot=<0 0 0.034 -0.9994>}
 spine_02 {parent=spine_01, loc=<7.2518 -0.035 0>, rot=<0 0 0.0238 -0.9997>}
 spine_03 {parent=spine_02, loc=<6.0104 0 0>, rot=<0 0 0.0149 -0.9999>}
 spine_04 {parent=spine_03, loc=<8.1024 0 0>, rot=<0 0 -0.0132 -0.9999>}
 spine_05 {parent=spine_04, loc=<10.0146 0 0>, rot=<0 0 -0.0015 -1>}
 neck_01 {parent=spine_05, loc=<11.3138 0.1529 0>, rot=<0 0 -0.2003 -0.9797>}
 neck_02 {parent=neck_01, loc=<4.5627 0 0>, rot=<0 0 0.0504 -0.9987>}
 head {parent=neck_02, loc=<5.3914 0 0>, rot=<0 0 0.115 -0.9934>}
 attach_cape {parent=spine_05, loc=<9.017 -11.714 0>, rot=<0.0616 -0.7044 -0.0616 -0.7044>}
 attach_backpack {parent=spine_05, loc=<-0.123 -13.473 0>, rot=<-0.0093 -0.707 0.0093 -0.707>}
 clavicle_l {parent=spine_05, loc=<4.3575 -0.1466 -2.3223>, rot=<0.0465 -0.7289 -0.009 -0.6829>}
 upperarm_l {parent=clavicle_l, loc=<15.5111 0 0>, rot=<0.0226 -0.3234 -0.0088 -0.9459>}
 lowerarm_l {parent=upperarm_l, loc=<26.6514 0 0>, rot=<0.0391 -0.0099 -0.2453 -0.9686>}
 hand_l {parent=lowerarm_l, loc=<22.3491 0 0>, rot=<-0.7081 -0.0507 0.0975 -0.6975>}
 thumb_01_l {parent=hand_l, loc=<1.3118 -0.3191 -2.459>, rot=<0.6546 -0.4042 0.0664 -0.6354>}
 thumb_02_l {parent=thumb_01_l, loc=<5.9336 -0.0448 0.0401>, rot=<-0.0037 0.0213 0.298 -0.9543>}
 thumb_03_l {parent=thumb_02_l, loc=<2.2223 0.3682 -0.0907>, rot=<0.0118 -0.013 0.0066 -0.9998>}
 index_metacarpal_l {parent=hand_l, loc=<1.6535 0.0183 -2.96>, rot=<0.0158 -0.0124 -0.0173 -0.9996>}
 index_01_l {parent=index_metacarpal_l, loc=<8.3008 0.0357 -0.0524>, rot=<-0.0044 0.0485 0.0924 -0.9945>}
 index_02_l {parent=index_01_l, loc=<3.3587 -0.3004 0.0452>, rot=<-0.002 0.0027 0.3225 -0.9466>}
 index_03_l {parent=index_02_l, loc=<2.5578 -0.123 -0.0252>, rot=<-0.002 0.0067 0.2558 -0.9667>}
 middle_metacarpal_l {parent=hand_l, loc=<2.1856 -0.1725 -0.5984>, rot=<0.006 0.0168 -0.0329 -0.9993>}
 middle_01_l {parent=middle_metacarpal_l, loc=<7.3275 -0.2447 0.0129>, rot=<-0.0208 0.0042 0.0858 -0.9961>}
 middle_02_l {parent=middle_01_l, loc=<4.0123 0.1401 0.1175>, rot=<-0.0218 0.0258 0.3512 -0.9357>}
 middle_03_l {parent=middle_02_l, loc=<2.6077 0.0534 0.1373>, rot=<0.0082 0.0188 0.1708 -0.9851>}
 ring_metacarpal_l {parent=hand_l, loc=<2.2074 -0.3625 1.3638>, rot=<-0.0284 0.0722 -0.0458 -0.9959>}
 ring_01_l {parent=ring_metacarpal_l, loc=<7.0295 -0.6197 -0.1446>, rot=<-0.0489 -0.0252 0.1892 -0.9804>}
 ring_02_l {parent=ring_01_l, loc=<3.6174 0.4647 -0.1027>, rot=<0 0 0.2806 -0.9598>}
 ring_03_l {parent=ring_02_l, loc=<2.4152 0.114 0.0199>, rot=<0 0 0.1189 -0.9929>}
 pinky_metacarpal_l {parent=hand_l, loc=<1.7616 -0.9753 3.3814>, rot=<-0.0397 0.0814 -0.069 -0.9935>}
 pinky_01_l {parent=pinky_metacarpal_l, loc=<7.1709 -0.4831 -0.1094>, rot=<-0.0654 -0.0035 0.2242 -0.9723>}
 pinky_02_l {parent=pinky_01_l, loc=<2.8274 0.0893 -0.1043>, rot=<0.0037 0.0011 0.2854 -0.9584>}
 pinky_03_l {parent=pinky_02_l, loc=<2.2046 0.0307 -0.0078>, rot=<-0.0002 0.0098 0.1445 -0.9895>}
 deform_wrist_dn_base_l {parent=hand_l, loc=<0 -2.2342 0>, rot=<0 0 0 -1>}
 deform_wrist_dn_l {parent=deform_wrist_dn_base_l, loc=<0 0 0>, rot=<0 0 0 1>}
 deform_wrist_up_base_l {parent=hand_l, loc=<0 2.2682 0>, rot=<0 0 0 -1>}
 deform_wrist_up_l {parent=deform_wrist_up_base_l, loc=<0 0 0>, rot=<0 0 0 1>}
 lowerarm_twist_02_l {parent=lowerarm_l, loc=<9.6199 0 0>, rot=<0 0 0 -1>}
 lowerarm_twist_01_l {parent=lowerarm_l, loc=<17.8549 0 0>, rot=<0 0 0 -1>}
 deform_elbow_in_base_l {parent=lowerarm_l, loc=<0.2112 5.259 0.8274>, rot=<0 0 0 -1>}
 deform_elbow_in_l {parent=deform_elbow_in_base_l, loc=<0 0 0>, rot=<0 0 0 1>}
 deform_elbow_out_base_l {parent=lowerarm_l, loc=<0 -4 0>, rot=<0 0 0 -1>}
 deform_elbow_out_l {parent=deform_elbow_out_base_l, loc=<0 0 0>, rot=<0 0 0 1>}
 upperarm_twist_01_l {parent=upperarm_l, loc=<0.0041 0 0>, rot=<0 0 0 -1>}
 deform_upperarm_bk_base_l {parent=upperarm_twist_01_l, loc=<1.0813 -7.2193 -1.9453>, rot=<0 0 0 1>}
 deform_upperarm_bk_l {parent=deform_upperarm_bk_base_l, loc=<0 0 0>, rot=<0 0 0 1>}
 deform_upperarm_fr_base_l {parent=upperarm_twist_01_l, loc=<-0.3691 7.0506 -2.765>, rot=<0 0 0 1>}
 deform_upperarm_fr_l {parent=deform_upperarm_fr_base_l, loc=<0 0 0>, rot=<0 0 0 1>}
 deform_upperarm_up_base_l {parent=upperarm_twist_01_l, loc=<-3.2216 -1.1391 3.3311>, rot=<0 0 0 1>}
 deform_upperarm_up_l {parent=deform_upperarm_up_base_l, loc=<0 0 0>, rot=<0 0 0 1>}
 upperarm_twist_02_l {parent=upperarm_l, loc=<13.7941 0 0>, rot=<0 0 0 -1>}
 base_ArmFur_l_01 {parent=upperarm_twist_02_l, loc=<17.9951 -4.2833 -0.3564>, rot=<-0.6277 -0.3039 0.0664 -0.7137>}
 dyn_ArmFur_l_01 {parent=base_ArmFur_l_01, loc=<0 0 0>, rot=<0 0 0 1>}
 base_ArmFur_l_02 {parent=upperarm_twist_02_l, loc=<16.4038 -5.9808 0.6857>, rot=<-0.674 0.0932 -0.2951 -0.6707>}
 dyn_ArmFur_l_02 {parent=base_ArmFur_l_02, loc=<0 0 0>, rot=<0 0 0 1>}
 base_ArmFur_l_03 {parent=upperarm_twist_02_l, loc=<13.0536 -6.8332 0.4868>, rot=<-0.0127 0.308 0.0659 -0.949>}
 dyn_ArmFur_l_03 {parent=base_ArmFur_l_03, loc=<0 0 0>, rot=<0 0 0 1>}
 base_ShoulderFur_l_02 {parent=clavicle_l, loc=<12.5145 -0.6706 6.2172>, rot=<-0.0953 -0.105 -0.1834 -0.9727>}
 dyn_ShoulderFur_l_02 {parent=base_ShoulderFur_l_02, loc=<0 0 0>, rot=<0 0 0 1>}
 base_ShoulderFur_l_01 {parent=clavicle_l, loc=<9.7274 0.7329 7.299>, rot=<0.0076 0.1703 -0.0841 -0.9818>}
 dyn_ShoulderFur_l_01 {parent=base_ShoulderFur_l_01, loc=<0 0 0>, rot=<0 0 0 1>}
 clavicle_r {parent=spine_05, loc=<4.3575 -0.1466 2.3223>, rot=<0.7289 0.0465 -0.6829 0.009>}
 upperarm_r {parent=clavicle_r, loc=<-15.5111 0 0>, rot=<0.0226 -0.3234 -0.0088 -0.9459>}
 lowerarm_r {parent=upperarm_r, loc=<-26.6514 0 0>, rot=<0.0391 -0.0099 -0.2453 -0.9686>}
 hand_r {parent=lowerarm_r, loc=<-22.3491 0 0>, rot=<-0.7081 -0.0507 0.0975 -0.6975>}
 thumb_01_r {parent=hand_r, loc=<-1.3118 0.3191 2.459>, rot=<0.6546 -0.4042 0.0664 -0.6354>}
 thumb_02_r {parent=thumb_01_r, loc=<-5.9336 0.0448 -0.0401>, rot=<-0.0037 0.0213 0.298 -0.9543>}
 thumb_03_r {parent=thumb_02_r, loc=<-2.2223 -0.3682 0.0907>, rot=<0.0118 -0.013 0.0066 -0.9998>}
 index_metacarpal_r {parent=hand_r, loc=<-1.6535 -0.0183 2.96>, rot=<0.0158 -0.0124 -0.0173 -0.9996>}
 index_01_r {parent=index_metacarpal_r, loc=<-8.3008 -0.0357 0.0524>, rot=<-0.0044 0.0485 0.0924 -0.9945>}
 index_02_r {parent=index_01_r, loc=<-3.3587 0.3004 -0.0452>, rot=<-0.002 0.0027 0.3225 -0.9466>}
 index_03_r {parent=index_02_r, loc=<-2.5578 0.123 0.0252>, rot=<-0.002 0.0067 0.2558 -0.9667>}
 middle_metacarpal_r {parent=hand_r, loc=<-2.1856 0.1725 0.5984>, rot=<0.006 0.0168 -0.0329 -0.9993>}
 middle_01_r {parent=middle_metacarpal_r, loc=<-7.3275 0.2447 -0.0129>, rot=<-0.0208 0.0042 0.0858 -0.9961>}
 middle_02_r {parent=middle_01_r, loc=<-4.0123 -0.1401 -0.1175>, rot=<-0.0218 0.0258 0.3512 -0.9357>}
 middle_03_r {parent=middle_02_r, loc=<-2.6077 -0.0534 -0.1373>, rot=<0.0082 0.0188 0.1708 -0.9851>}
 ring_metacarpal_r {parent=hand_r, loc=<-2.2074 0.3625 -1.3638>, rot=<-0.0284 0.0722 -0.0458 -0.9959>}
 ring_01_r {parent=ring_metacarpal_r, loc=<-7.0295 0.6197 0.1446>, rot=<-0.0489 -0.0252 0.1892 -0.9804>}
 ring_02_r {parent=ring_01_r, loc=<-3.6174 -0.4647 0.1027>, rot=<0 0 0.2806 -0.9598>}
 ring_03_r {parent=ring_02_r, loc=<-2.4152 -0.114 -0.0199>, rot=<0 0 0.1189 -0.9929>}
 pinky_metacarpal_r {parent=hand_r, loc=<-1.7616 0.9753 -3.3814>, rot=<-0.0397 0.0814 -0.069 -0.9935>}
 pinky_01_r {parent=pinky_metacarpal_r, loc=<-7.1709 0.4831 0.1094>, rot=<-0.0654 -0.0035 0.2242 -0.9723>}
 pinky_02_r {parent=pinky_01_r, loc=<-2.8274 -0.0893 0.1043>, rot=<0.0037 0.0011 0.2854 -0.9584>}
 pinky_03_r {parent=pinky_02_r, loc=<-2.2046 -0.0307 0.0078>, rot=<-0.0002 0.0098 0.1445 -0.9895>}
 deform_wrist_dn_base_r {parent=hand_r, loc=<-0.0004 2.2345 -0.0001>, rot=<0 0 0 -1>}
 deform_wrist_dn_r {parent=deform_wrist_dn_base_r, loc=<0 0 0>, rot=<0 0 0 1>}
 deform_wrist_up_base_r {parent=hand_r, loc=<0.0003 -2.2684 0.0001>, rot=<0 0 0 -1>}
 deform_wrist_up_r {parent=deform_wrist_up_base_r, loc=<0 0 0>, rot=<0 0 0 1>}
 lowerarm_twist_02_r {parent=lowerarm_r, loc=<-9.6199 0 0>, rot=<0 0 0 -1>}
 lowerarm_twist_01_r {parent=lowerarm_r, loc=<-17.8549 0 0>, rot=<0 0 0 -1>}
 deform_elbow_in_base_r {parent=lowerarm_r, loc=<-0.2111 -5.2591 -0.8275>, rot=<0 0 0 -1>}
 deform_elbow_in_r {parent=deform_elbow_in_base_r, loc=<-0.0001 0 0.0002>, rot=<0 0 0 1>}
 deform_elbow_out_base_r {parent=lowerarm_r, loc=<0 4 -0.0001>, rot=<0 0 0 -1>}
 deform_elbow_out_r {parent=deform_elbow_out_base_r, loc=<0 0 0>, rot=<0 0 0 1>}
 upperarm_twist_01_r {parent=upperarm_r, loc=<-0.0041 0 0>, rot=<0 0 0 -1>}
 deform_upperarm_bk_base_r {parent=upperarm_twist_01_r, loc=<-1.0814 7.2192 1.9455>, rot=<0 0 0 1>}
 deform_upperarm_bk_r {parent=deform_upperarm_bk_base_r, loc=<0 0 0>, rot=<0 0 0 1>}
 deform_upperarm_fr_base_r {parent=upperarm_twist_01_r, loc=<0.3689 -7.0506 2.7652>, rot=<0 0 0 1>}
 deform_upperarm_fr_r {parent=deform_upperarm_fr_base_r, loc=<0 0 0>, rot=<0 0 0 1>}
 deform_upperarm_up_base_r {parent=upperarm_twist_01_r, loc=<3.2214 1.1391 -3.3308>, rot=<0 0 0 1>}
 deform_upperarm_up_r {parent=deform_upperarm_up_base_r, loc=<0 0 0>, rot=<0 0 0 -1>}
 upperarm_twist_02_r {parent=upperarm_r, loc=<-13.7941 0 0>, rot=<0 0 0 -1>}
 base_ArmFur_r_01 {parent=upperarm_twist_02_r, loc=<-17.9952 4.2833 0.3563>, rot=<-0.0664 -0.7137 -0.6277 0.3039>}
 dyn_ArmFur_r_01 {parent=base_ArmFur_r_01, loc=<0 0 0>, rot=<0 0 0 1>}
 base_ArmFur_r_03 {parent=upperarm_twist_02_r, loc=<-13.0535 6.8332 -0.487>, rot=<-0.1712 0.6621 0.68 0.2644>}
 dyn_ArmFur_r_03 {parent=base_ArmFur_r_03, loc=<0 0 0>, rot=<0 0 0 1>}
 base_ArmFur_r_02 {parent=upperarm_twist_02_r, loc=<-16.4039 5.9808 -0.6856>, rot=<0.2614 -0.7375 -0.6164 -0.0888>}
 dyn_ArmFur_r_02 {parent=base_ArmFur_r_02, loc=<0 0 0>, rot=<0 0 0 1>}
 base_ShoulderFur_r_01 {parent=clavicle_r, loc=<-13.3094 -1.5336 -5.966>, rot=<-0.0849 -0.9933 0.0689 0.0368>}
 dyn_ShoulderFur_r_01 {parent=base_ShoulderFur_r_01, loc=<0 0 0>, rot=<0 0 0 1>}
 base_ShoulderFur_r_02 {parent=clavicle_r, loc=<-7.843 -0.8408 -6.8821>, rot=<0.1045 -0.9089 0.1176 -0.3861>}
 dyn_ShoulderFur_r_02 {parent=base_ShoulderFur_r_02, loc=<0 0 0>, rot=<0 0 0 1>}
 base_ShoulderFur_r_03 {parent=clavicle_r, loc=<-10.0439 1.7155 -6.4981>, rot=<0.1334 -0.9494 -0.2504 -0.1347>}
 dyn_ShoulderFur_r_03 {parent=base_ShoulderFur_r_03, loc=<0 0 0>, rot=<0 0 0 1>}
 deform_pec_aim_base_l {parent=spine_04, loc=<7.8732 12.6388 -2.4123>, rot=<0.0102 -0.4602 0.3979 -0.7936>}
 deform_pec_aim_l {parent=deform_pec_aim_base_l, loc=<0 0 0>, rot=<0 0 0 1>}
 deform_pec_base_l {parent=deform_pec_aim_l, loc=<10.766 0 0>, rot=<0 0 0 1>}
 deform_pec_l {parent=deform_pec_base_l, loc=<0.055 0 0>, rot=<0 0 0 1>}
 deform_pec_aim_base_r {parent=spine_04, loc=<7.8727 12.6388 2.4123>, rot=<-0.4602 -0.0102 0.7936 0.3979>}
 deform_pec_aim_r {parent=deform_pec_aim_base_r, loc=<0 0 0>, rot=<0 0 0 1>}
 deform_pec_base_r {parent=deform_pec_aim_r, loc=<-10.7662 -0.0006 -0.0007>, rot=<0 0 0 1>}
 deform_pec_r {parent=deform_pec_base_r, loc=<0 0 0>, rot=<0 0 0 1>}
 thigh_l {parent=pelvis, loc=<-2.5693 -0.0109 -7.7501>, rot=<0.0012 0.0572 0.0155 -0.9982>}
 calf_l {parent=thigh_l, loc=<-39.0484 0 0>, rot=<0.0006 -0.0161 -0.0391 -0.9991>}
 foot_l {parent=calf_l, loc=<-42.561 0.1077 0.4292>, rot=<0.0418 -0.0455 0.0343 -0.9975>}
 ball_l {parent=foot_l, loc=<-6.1599 13.8264 0.6806>, rot=<0.0001 -0.0001 -0.7063 -0.7079>}
 deform_foot_base_bk_l {parent=foot_l, loc=<0 0 0>, rot=<0.1073 0 0 -0.9942>}
 deform_foot_bk_l {parent=deform_foot_base_bk_l, loc=<0 -4.9203 0>, rot=<0 0 0 1>}
 calf_twist_02_l {parent=calf_l, loc=<-18.8064 -0.0967 0.0397>, rot=<0.0237 0.0095 -0.0035 -0.9997>}
 calf_twist_01_l {parent=calf_l, loc=<-35.813 -0.0967 0.0397>, rot=<0.0237 0.0095 -0.0035 -0.9997>}
 deform_knee_in_base_l {parent=calf_l, loc=<1.298 -5.2957 -0.0496>, rot=<0 0 0 -1>}
 deform_knee_in_l {parent=deform_knee_in_base_l, loc=<0 0 0>, rot=<0 0 0 1>}
 deform_knee_out_base_l {parent=calf_l, loc=<0 4 0>, rot=<0 0 0 -1>}
 deform_knee_out_l {parent=deform_knee_out_base_l, loc=<0 0 0>, rot=<0 0 0 1>}
 foot_l_null {parent=calf_l, type=Null, loc=<-42.561 0.1077 0.4292>, rot=<0.0418 -0.0455 0.0343 -0.9975>}
 thigh_twist_01_l {parent=thigh_l, loc=<-21.6303 0 0>, rot=<0.0013 0 0 -1>}
 deform_groin_base_l {parent=thigh_l, loc=<-2.1908 8.1142 -1.7442>, rot=<0 0 0 -1>}
 deform_groin_l {parent=deform_groin_base_l, loc=<0 0 0>, rot=<0 0 0 1>}
 base_legFur_l_01 {parent=thigh_l, loc=<-22.6988 2.4204 -4.1031>, rot=<-0.1994 -0.9321 -0.1925 -0.233>}
 dyn_legFur_l_01 {parent=base_legFur_l_01, loc=<0 0 0>, rot=<0 0 0 1>}
 base_legFur_l_02 {parent=thigh_l, loc=<-23.9392 -2.3422 -5.8265>, rot=<0.059 -0.8179 -0.0577 -0.5694>}
 dyn_legFur_l_02 {parent=base_legFur_l_02, loc=<0 0 0>, rot=<0 0 0 1>}
 base_legFur_l_03 {parent=thigh_l, loc=<-22.9275 -5.383 -2.2536>, rot=<0.1274 -0.8409 0.4141 -0.3243>}
 dyn_legFur_l_03 {parent=base_legFur_l_03, loc=<0 0 0>, rot=<0 0 0 1>}
 base_legFur_l_04 {parent=thigh_l, loc=<-23.1498 -6.0375 2.1435>, rot=<0.3311 -0.6275 0.6573 -0.254>}
 dyn_legFur_l_04 {parent=base_legFur_l_04, loc=<0 0 0>, rot=<0 0 0 1>}
 base_legFur_l_05 {parent=thigh_l, loc=<-22.8629 -1.9377 4.6142>, rot=<0.3402 0.1036 0.9337 -0.0408>}
 dyn_legFur_l_05 {parent=base_legFur_l_05, loc=<0 0 0>, rot=<0 0 0 1>}
 base_legFur_l_06 {parent=thigh_l, loc=<-22.7766 2.1884 4.6735>, rot=<0.5581 0.1301 0.8195 0.0039>}
 dyn_legFur_l_06 {parent=base_legFur_l_06, loc=<0 0 0>, rot=<0 0 0 1>}
 base_legFur_l_07 {parent=thigh_l, loc=<-22.4947 4.9889 2.2151>, rot=<0.3326 0.37 0.8415 0.2107>}
 dyn_legFur_l_07 {parent=base_legFur_l_07, loc=<0 0 0>, rot=<0 0 0 1>}
 base_legFur_l_08 {parent=thigh_l, loc=<-22.6406 5.1756 -1.5187>, rot=<-0.3096 -0.6758 -0.5626 -0.3619>}
 dyn_legFur_l_08 {parent=base_legFur_l_08, loc=<0 0 0>, rot=<0 0 0 1>}
 thigh_r {parent=pelvis, loc=<-2.5693 -0.0109 7.7501>, rot=<0.0572 -0.0012 0.9982 0.0155>}
 calf_r {parent=thigh_r, loc=<39.0484 0 0>, rot=<0.0006 -0.0161 -0.0391 -0.9991>}
 foot_r {parent=calf_r, loc=<42.561 -0.1077 -0.4292>, rot=<0.0418 -0.0455 0.0343 -0.9975>}
 ball_r {parent=foot_r, loc=<6.1599 -13.8264 -0.6806>, rot=<0.0001 -0.0001 -0.7063 -0.7079>}
 deform_foot_base_bk_r {parent=foot_r, loc=<0 0 0>, rot=<-0.0057 -0.0905 0.9959 0>}
 deform_foot_bk_r {parent=deform_foot_base_bk_r, loc=<0.0051 -4.8398 -0.8868>, rot=<-0.1073 0 0 -0.9942>}
 calf_twist_02_r {parent=calf_r, loc=<18.8064 0.0967 -0.0397>, rot=<0.0237 0.0095 -0.0035 -0.9997>}
 calf_twist_01_r {parent=calf_r, loc=<35.813 0.0967 -0.0397>, rot=<0.0237 0.0095 -0.0035 -0.9997>}
 deform_knee_in_base_r {parent=calf_r, loc=<-1.298 5.2957 0.0496>, rot=<0 0 0 -1>}
 deform_knee_in_r {parent=deform_knee_in_base_r, loc=<0 0 0>, rot=<0 0 0 1>}
 deform_knee_out_base_r {parent=calf_r, loc=<0 -4 0>, rot=<0 0 0 -1>}
 deform_knee_out_r {parent=deform_knee_out_base_r, loc=<0 0 0>, rot=<0 0 0 1>}
 foot_r_null {parent=calf_r, type=Null, loc=<42.561 -0.1077 -0.4292>, rot=<0.0418 -0.0455 0.0343 -0.9975>}
 thigh_twist_01_r {parent=thigh_r, loc=<21.6303 0 0>, rot=<0.0013 0 0 -1>}
 deform_groin_base_r {parent=thigh_r, loc=<2.1908 -8.1142 1.7442>, rot=<0 0 0 -1>}
 deform_groin_r {parent=deform_groin_base_r, loc=<0 0 0>, rot=<0 0 0 1>}
 base_legwFur_r_01 {parent=thigh_r, loc=<23.2212 -2.9797 4.579>, rot=<-0.3788 0.334 0.2185 -0.835>}
 dyn_legwFur_r_01 {parent=base_legwFur_r_01, loc=<0 0 0>, rot=<0 0 0 1>}
 base_legwFur_r_02 {parent=thigh_r, loc=<23.0921 1.2957 5.5589>, rot=<-0.1305 0.5136 -0.0105 -0.848>}
 dyn_legwFur_r_02 {parent=base_legwFur_r_02, loc=<0 0 0>, rot=<0 0 0 1>}
 base_legwFur_r_03 {parent=thigh_r, loc=<22.6924 4.5033 3.2562>, rot=<0.2583 0.3306 -0.1634 -0.8929>}
 dyn_legwFur_r_03 {parent=base_legwFur_r_03, loc=<0 0 0>, rot=<0 0 0 1>}
 base_legwFur_r_04 {parent=thigh_r, loc=<23.5927 5.4185 -0.8722>, rot=<0.7134 0.2975 -0.4247 -0.4714>}
 dyn_legwFur_r_04 {parent=base_legwFur_r_04, loc=<0 0 0>, rot=<0 0 0 1>}
 base_legwFur_r_05 {parent=thigh_r, loc=<23.0355 2.5972 -4.3755>, rot=<0.9089 0.1398 -0.3091 -0.2424>}
 dyn_legwFur_r_05 {parent=base_legwFur_r_05, loc=<0 0 0>, rot=<0 0 0 1>}
 base_legwFur_r_06 {parent=thigh_r, loc=<22.7247 -1.7259 -5.6078>, rot=<0.828 -0.0699 -0.5082 -0.2265>}
 dyn_legwFur_r_06 {parent=base_legwFur_r_06, loc=<0 0 0>, rot=<0 0 0 1>}
 base_legwFur_r_07 {parent=thigh_r, loc=<22.6169 -4.9198 -3.6271>, rot=<0.8153 -0.1558 -0.491 0.2645>}
 dyn_legwFur_r_07 {parent=base_legwFur_r_07, loc=<0 0 0>, rot=<0 0 0 1>}
 base_legwFur_r_08 {parent=thigh_r, loc=<22.8263 -5.879 0.2105>, rot=<-0.6465 0.3015 0.3363 -0.6148>}
 dyn_legwFur_r_08 {parent=base_legwFur_r_08, loc=<0 0 0>, rot=<0 0 0 1>}
 deform_glute_base_l {parent=pelvis, loc=<-8.8064 -8.8222 -8.4287>, rot=<0.0275 0.0501 -0.4472 -0.8926>}
 deform_glute_l {parent=deform_glute_base_l, loc=<0 0 0>, rot=<0 0 0 1>}
 deform_glute_base_r {parent=pelvis, loc=<-8.8062 -8.8222 8.4287>, rot=<0.0501 -0.0275 0.8926 -0.4472>}
 deform_glute_r {parent=deform_glute_base_r, loc=<0 0 0>, rot=<0 0 0 1>}
 base_tail_01 {parent=pelvis, loc=<-0.7928 -16.009 0.0494>, rot=<0.6858 -0.1725 0.1725 -0.6858>}
 dyn_tail_01 {parent=base_tail_01, loc=<0 0 0>, rot=<0 0 0 1>}
 dyn_tailfur_01 {parent=dyn_tail_01, loc=<11.5835 -0.8175 0.9943>, rot=<0 0.3982 0 -0.9173>}
 dyn_tailfur_03 {parent=dyn_tail_01, loc=<9.2486 -7.6398 1.5375>, rot=<0.2025 0.2626 -0.0294 -0.9429>}
 dyn_tailfur_02 {parent=dyn_tail_01, loc=<8.6572 5.6635 1.8151>, rot=<0.0495 0.1963 0.0368 -0.9786>}
 thigh_r_null {parent=pelvis, type=Null, loc=<-2.5693 -0.0109 7.7501>, rot=<0.0572 -0.0012 0.9982 0.0155>}
 thigh_l_null {parent=pelvis, type=Null, loc=<-2.5693 -0.0109 -7.7501>, rot=<0.0012 0.0572 0.0155 -0.9982>}
]
variables [
 BackBiling {type=bool, default=False}
 Custom_Emote {type=double, default=0}
 Front End {type=bool, default=False}
 ParachuteOpened {type=bool, default=False}
]
