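# Physics asset for the left shorts leg of the male medium NotchGnarl Aloud player cosmetic, defining secondary-motion bodies for the fabric.
# Front, back, inner and outer chains each link a tiny base capsule to two dynamic capsules through constraints with locked twist and narrow swing limits.
# An unconstrained thigh capsule supplies the leg's shape, and the lower front segment carries an extra angled capsule.
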PhysicsAsset /BRCosmetics/Characters/Player/Male/Medium/Bodies/M_MED_NotchGnarl_Aloud/Meshes/M_MED_NotchGnarl_Aloud_ShortsLeft_Physics
bodies {
 base_shorts_bk_l [{shapeType=Capsule, length=0.505, radius=0.505, shapeName=base_shorts_bk_l_capsule, rotation=<rot 0 -90 0>}]
 base_shorts_fr_l [{shapeType=Capsule, length=0.505, radius=0.505, shapeName=base_shorts_fr_l_capsule, rotation=<rot 0 -90 0>}]
 base_shorts_general_l [{shapeType=Capsule, length=0.505, radius=0.505, shapeName=base_shorts_general_l_capsule, rotation=<rot 0 -90 0>}]
 base_shorts_in_l [{shapeType=Capsule, length=0.505, radius=0.505, shapeName=base_shorts_in_l_capsule, rotation=<rot 0 -90 0>}]
 base_shorts_out_l [{shapeType=Capsule, length=0.505, radius=0.505, shapeName=base_shorts_out_l_capsule, rotation=<rot 0 -90 0>}]
 dyn_shorts_bk_01_l [{shapeType=Capsule, length=10, radius=4, shapeName=dyn_shorts_bk_01_l_capsule, center=<5.6319 0 1.9745>, rotation=<rot 0 0 90>}]
 dyn_shorts_bk_02_l [{shapeType=Capsule, length=10, radius=4, shapeName=dyn_shorts_bk_02_l_capsule, center=<5.6319 0 1.4357>, rotation=<rot 0 0 90>}]
 dyn_shorts_fr_01_l [{shapeType=Capsule, length=11.6518, radius=4, shapeName=dyn_shorts_fr_01_l_capsule, center=<5.6319 0 4.1053>, rotation=<rot 0 0 90>}]
 dyn_shorts_fr_02_l [
  {shapeType=Capsule, length=11.6518, radius=4, shapeName=dyn_shorts_fr_02_l_capsule, center=<5.6319 0 4.0067>, rotation=<rot 0 0 90>}
  {
   shapeType=Capsule
   length=8.3658
   radius=4.89
   shapeName=dyn_shorts_fr_02_l_capsule
   center=<4.2022 10.4183 0.4233>
   rotation=<rot -2.1697 2.9687 -45.0167>
  }
 ]
 dyn_shorts_in_01_l [{shapeType=Capsule, length=10, radius=4, shapeName=dyn_shorts_in_01_l_capsule, center=<5.6477 0 3.8679>, rotation=<rot 0 0 90>}]
 dyn_shorts_in_02_l [{shapeType=Capsule, length=10, radius=4, shapeName=dyn_shorts_in_02_l_capsule, center=<5.6477 0 3.1652>, rotation=<rot 0 0 90>}]
 dyn_shorts_out_01_l [{shapeType=Capsule, length=11.5672, radius=4, shapeName=dyn_shorts_out_01_l_capsule, center=<5.6319 0 3.6112>, rotation=<rot 0 0 90>}]
 dyn_shorts_out_02_l [{shapeType=Capsule, length=11.5672, radius=4, shapeName=dyn_shorts_out_02_l_capsule, center=<5.6319 0 4.9993>, rotation=<rot 0 0 90>}]
 thigh_l [{shapeType=Capsule, length=26.8585, radius=4.4079, shapeName=thigh_l_capsule, center=<-24.4914 0.2273 0.3047>, rotation=<rot 90 0 0>}]
}
constraints [
 {
  bone1Name=dyn_shorts_bk_01_l
  bone2Name=base_shorts_bk_l
  swing1LimitDegrees=5
  swing2LimitDegrees=10
  twistLimitDegrees=45
  twistMotion=Locked
 }
 {
  bone1Name=dyn_shorts_bk_02_l
  bone2Name=dyn_shorts_bk_01_l
  swing1LimitDegrees=5
  swing2LimitDegrees=10
  twistLimitDegrees=45
  twistMotion=Locked
 }
 {
  bone1Name=dyn_shorts_out_02_l
  bone2Name=dyn_shorts_out_01_l
  swing1LimitDegrees=5
  swing2LimitDegrees=10
  twistLimitDegrees=45
  twistMotion=Locked
 }
 {
  bone1Name=dyn_shorts_fr_01_l
  bone2Name=base_shorts_fr_l
  swing1LimitDegrees=5
  swing2LimitDegrees=10
  twistLimitDegrees=45
  twistMotion=Locked
 }
 {
  bone1Name=dyn_shorts_fr_02_l
  bone2Name=dyn_shorts_fr_01_l
  swing1LimitDegrees=5
  swing2LimitDegrees=10
  twistLimitDegrees=45
  twistMotion=Locked
 }
 {
  bone1Name=dyn_shorts_out_01_l
  bone2Name=base_shorts_out_l
  swing1LimitDegrees=5
  swing2LimitDegrees=10
  twistLimitDegrees=45
  twistMotion=Locked
 }
 {
  bone1Name=dyn_shorts_in_01_l
  bone2Name=base_shorts_in_l
  swing1LimitDegrees=5
  swing2LimitDegrees=10
  twistLimitDegrees=45
  twistMotion=Locked
 }
 {
  bone1Name=dyn_shorts_in_02_l
  bone2Name=dyn_shorts_in_01_l
  swing1LimitDegrees=5
  swing2LimitDegrees=10
  twistLimitDegrees=45
  twistMotion=Locked
 }
]
