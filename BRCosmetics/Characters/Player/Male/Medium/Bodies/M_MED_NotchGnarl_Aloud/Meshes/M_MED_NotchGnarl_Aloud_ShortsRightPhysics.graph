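# Physics asset for the right shorts leg of the NotchGnarl_Aloud medium male cosmetic body, set up for secondary cloth-like motion through bones.
# Four two-segment capsule chains (front, back, inner, outer) each hang from a tiny base capsule, with constraints that lock twist and set narrow swing angles.
# A long thigh capsule and a general base capsule have no constraints, the thigh one apparently serving as a leg collider for the chains.
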PhysicsAsset /BRCosmetics/Characters/Player/Male/Medium/Bodies/M_MED_NotchGnarl_Aloud/Meshes/M_MED_NotchGnarl_Aloud_ShortsRightPhysics
bodies {
 base_shorts_bk_r [{shapeType=Capsule, length=0.505, radius=0.505, shapeName=base_shorts_bk_r_capsule, rotation=<rot 0 -90 180>}]
 base_shorts_fr_r [{shapeType=Capsule, length=0.505, radius=0.505, shapeName=base_shorts_fr_r_capsule, rotation=<rot 0 -90 180>}]
 base_shorts_general_r [{shapeType=Capsule, length=0.505, radius=0.505, shapeName=base_shorts_general_r_capsule, rotation=<rot 0 -90 180>}]
 base_shorts_in_r [{shapeType=Capsule, length=0.505, radius=0.505, shapeName=base_shorts_in_r_capsule, rotation=<rot 0 -90 180>}]
 base_shorts_out_r [{shapeType=Capsule, length=0.505, radius=0.505, shapeName=base_shorts_out_r_capsule, rotation=<rot 0 -90 180>}]
 dyn_shorts_bk_01_r [{shapeType=Capsule, length=10, radius=4, shapeName=dyn_shorts_bk_01_r_capsule, center=<5.7292 0 2.2416>, rotation=<rot 0 0 -90>}]
 dyn_shorts_bk_02_r [{shapeType=Capsule, length=10, radius=4, shapeName=dyn_shorts_bk_02_r_capsule, center=<5.7292 0 2.7804>, rotation=<rot 0 0 -90>}]
 dyn_shorts_fr_01_r [{shapeType=Capsule, length=11.6518, radius=4, shapeName=dyn_shorts_fr_01_r_capsule, center=<5.7292 0 2.962>, rotation=<rot 0 0 -90>}]
 dyn_shorts_fr_02_r [{shapeType=Capsule, length=11.6518, radius=4, shapeName=dyn_shorts_fr_02_r_capsule, center=<5.7292 0 3.0607>, rotation=<rot 0 0 -90>}]
 dyn_shorts_in_01_r [{shapeType=Capsule, length=10, radius=4, shapeName=dyn_shorts_in_01_r_capsule, center=<5.7134 0 2.6235>, rotation=<rot 0 0 -90>}]
 dyn_shorts_in_02_r [{shapeType=Capsule, length=10, radius=4, shapeName=dyn_shorts_in_02_r_capsule, center=<5.7134 0 2.8332>, rotation=<rot 0 0 -90>}]
 dyn_shorts_out_01_r [{shapeType=Capsule, length=11.5672, radius=4, shapeName=dyn_shorts_out_01_r_capsule, center=<5.7292 0 2.5122>, rotation=<rot 0 0 -90>}]
 dyn_shorts_out_02_r [{shapeType=Capsule, length=11.5672, radius=4, shapeName=dyn_shorts_out_02_r_capsule, center=<5.7292 0 3.6525>, rotation=<rot 0 0 -90>}]
 thigh_r [{shapeType=Capsule, length=26.8585, radius=4.4079, shapeName=thigh_r_capsule, center=<24.4914 0.086 -0.3047>, rotation=<rot 90 180 0>}]
}
constraints [
 {
  bone1Name=dyn_shorts_bk_01_r
  bone2Name=base_shorts_bk_r
  swing1LimitDegrees=5
  swing2LimitDegrees=10
  twistLimitDegrees=45
  twistMotion=Locked
 }
 {
  bone1Name=dyn_shorts_bk_02_r
  bone2Name=dyn_shorts_bk_01_r
  swing1LimitDegrees=5
  swing2LimitDegrees=10
  twistLimitDegrees=45
  twistMotion=Locked
 }
 {
  bone1Name=dyn_shorts_out_02_r
  bone2Name=dyn_shorts_out_01_r
  swing1LimitDegrees=5
  swing2LimitDegrees=10
  twistLimitDegrees=45
  twistMotion=Locked
 }
 {
  bone1Name=dyn_shorts_fr_01_r
  bone2Name=base_shorts_fr_r
  swing1LimitDegrees=5
  swing2LimitDegrees=10
  twistLimitDegrees=45
  twistMotion=Locked
 }
 {
  bone1Name=dyn_shorts_fr_02_r
  bone2Name=dyn_shorts_fr_01_r
  swing1LimitDegrees=5
  swing2LimitDegrees=10
  twistLimitDegrees=45
  twistMotion=Locked
 }
 {
  bone1Name=dyn_shorts_out_01_r
  bone2Name=base_shorts_out_r
  swing1LimitDegrees=5
  swing2LimitDegrees=10
  twistLimitDegrees=45
  twistMotion=Locked
 }
 {
  bone1Name=dyn_shorts_in_01_r
  bone2Name=base_shorts_in_r
  swing1LimitDegrees=5
  swing2LimitDegrees=10
  twistLimitDegrees=45
  twistMotion=Locked
 }
 {
  bone1Name=dyn_shorts_in_02_r
  bone2Name=dyn_shorts_in_01_r
  swing1LimitDegrees=5
  swing2LimitDegrees=10
  twistLimitDegrees=45
  twistMotion=Locked
 }
]
